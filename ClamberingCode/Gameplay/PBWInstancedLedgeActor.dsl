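; Blueprint actor for instanced ledges in the clambering system, holding ledge instances in an Instanced Static Mesh component.
; When an instance is added, it writes the current game time and the ledge's world or local transform location into that instance's custom data floats, making both available per instance to the material.
; The remove-instance, BeginPlay, Tick, overlap and construction script entries contain no logic.
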(event EventBeginPlay)

(event Collision|EventActorBeginOverlap
  (params (OtherActor :type Actor%20Object%20Reference)))

(event EventTick
  (params (DeltaSeconds :type Float%20%28single-precision%29)))

(event EventOnAddInstance
  (params (LedgeTransform :type Transform%20%28by%20ref%29) (InstanceIndex :type Integer))
  (bind _location (Math|Transform|BreakTransform LedgeTransform))
  (bind (_location_x _location_y _location_z) (Math|Vector|BreakVector _location))
  (bind _instancedstaticmeshcomponent (Variables|InstancedLedgeActor|GetInstancedStaticMeshComponent))
  (Components|InstancedStaticMesh|SetCustomDataValue (Variables|InstancedLedgeActor|GetInstancedStaticMeshComponent) InstanceIndex 0 (Utilities|Time|GetGameTimeinSeconds) false)
  (Components|InstancedStaticMesh|SetCustomDataValue _instancedstaticmeshcomponent InstanceIndex 9 _location_x false)
  (Components|InstancedStaticMesh|SetCustomDataValue _instancedstaticmeshcomponent InstanceIndex 10 _location_y false)
  (Components|InstancedStaticMesh|SetCustomDataValue _instancedstaticmeshcomponent InstanceIndex 11 _location_z false))

(event EventOnRemoveInstance
  (params (LedgeTransform :type Transform%20%28by%20ref%29) (InstanceIndex :type Integer)))

(fn ConstructionScript)
